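# Translucent unlit material for the inner glider effect of the Nebula character, usable on Niagara sprites, mesh particles and skeletal meshes.
# A screen-space galaxy function supplies emissive, with particle or time-of-day color available to add.
# Panned noise and a sine-pulsed ramp form a flame mask, alongside view-angle and depth-fade terms.
# Vertex offset scrolls a distortion texture, scales intensity by camera distance and shading path, and optionally adds a Niagara-driven bend.
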Material /Game/Characters/Player/Male/Medium/Bodies/M_MED_Nebula/FX/FX_Materials/M_FX_Nebula_GliderInner
graph {
 properties {
  bUsedWithMeshDeformer=true
  bUsedWithNiagaraMeshParticles=true
  bUsedWithNiagaraSprites=true
  bUsedWithSkeletalMesh=true
  blendMode=BLEND_Translucent
  pixelDepthOffsetMode=PDOM_Legacy
  shadingModel=MSM_Unlit
  usedShadingModels=Unlit
 }
 nodes [
  n1 = DynamicParameter() {desc=Parameters set by Niagara, parameterIndex=1, paramNames [X,Y,Z,WPOIntensity]}
  n2 = Clamp(
   None=Subtract(
    A=MaterialFunctionCall(X=@n1.X, Y=@n1.X, Z=@n1.X) {materialFunction=&MakeFloat3}.Result
    B=ParticlePositionWS()
   )
   Min=ScalarParameter() {defaultValue=-100, group=WPO, parameterName=Clamp_min}
   Max=ScalarParameter() {defaultValue=100, group=WPO, parameterName=Clamp_max}
  ) {maxDefault=1000, minDefault=-1000}
  n3 = OneMinus(None=TextureCoordinate())
  n4 = Multiply(
   A=Power(Base=ComponentMask(None=@n3) {g=true}, Exp=ScalarParameter() {defaultValue=3, group=WPO, parameterName=WPOBendPower})
   B=@n1.WPOIntensity
  )
  n5 = OneMinus(None=MaterialFunctionCall(Distance=Constant() {r=3000}) {materialFunction=&Near_Camera_Fade}.Result)
  n6 = VectorParameter() {group=WPO, parameterName=WPO_UVScale, defaultValue=<rgba 0 1 0 0>}
  n7 = Add(A=Multiply(A=TextureCoordinate(), B=ComponentMask(None=@n6.RGB) {g=true, r=true}), B=AppendVector(A=Constant(), B=VertexColor().R)) {constB=0}
  n8 = ComponentMask(None=VectorParameter() {group=WPO, parameterName=WPO_Speed, defaultValue=<rgba 0 0.5 0 0>}.RGB) {g=true, r=true}
  n9 = LinearInterpolate(Alpha=TextureSample(UVs=Panner(Coordinate=@n7, Speed=@n8) {speedY=0.5}) {samplerType=SAMPLERTYPE_Normal, texture=&T_Weap_Distort_N_02}.R) {constB=1.5}
  n10 = OneMinus(None=ComponentMask(None=TextureCoordinate()) {g=true})
  n11 = Saturate(None=Multiply(A=Power(Base=@n10) {constExponent=1.5}))
  n12 = Transform(None=Multiply(A=Multiply(A=@n9, B=@n11), B=VectorParameter() {group=WPO, parameterName=WPO_direction, defaultValue=<rgba 1 0 0 1>}.RGB))
  n13 = Multiply(
   A=@n12
   B=LinearInterpolate(
    A=ScalarParameter() {defaultValue=20, group=WPO, parameterName=Glider_Lod_low_WPO_intensity}
    B=ScalarParameter() {defaultValue=80, group=WPO, parameterName=Glider_WPO_intensity}
    Alpha=Saturate(None=@n5)
   )
  ) {constB=80}
  n14 = OneMinus(None=MaterialFunctionCall(Distance=Constant() {r=1000}) {materialFunction=&Near_Camera_Fade}.Result)
  n15 = Multiply(
   A=@n12
   B=LinearInterpolate(
    A=ScalarParameter() {defaultValue=15, group=WPO, parameterName=Glider_Moble_Lod_low_WPO_intensity}
    B=ScalarParameter() {defaultValue=80, group=WPO, parameterName=Glider_WPO_intensity}
    Alpha=Saturate(None=@n14)
   )
  )
  n16 = OneMinus(None=MaterialFunctionCall(Distance=Constant() {r=1700}) {materialFunction=&Near_Camera_Fade}.Result)
  n17 = Multiply(
   A=@n12
   B=LinearInterpolate(
    A=ScalarParameter() {defaultValue=20, group=WPO, parameterName=Lod_low_WPO_intensity}
    B=ScalarParameter() {defaultValue=80, group=WPO, parameterName=WPO_intensity}
    Alpha=Saturate(None=@n16)
   )
  ) {constB=80}
  n18 = OneMinus(None=MaterialFunctionCall(Distance=Constant() {r=500}) {materialFunction=&Near_Camera_Fade}.Result)
  n19 = Multiply(
   A=@n12
   B=LinearInterpolate(
    A=ScalarParameter() {defaultValue=15, group=WPO, parameterName=Moble_Lod_low_WPO_intensity}
    B=ScalarParameter() {defaultValue=80, group=WPO, parameterName=WPO_intensity}
    Alpha=Saturate(None=@n18)
   )
  )
  Add(
   A=StaticSwitchParameter(True=LinearInterpolate(B=@n2, Alpha=@n4), False=Constant()) {group=WPO, parameterName=Use_Interpolate WPO}
   B=StaticSwitchParameter(
    True=ShadingPathSwitch(Default=@n13, Deferred=@n13, Forward=@n13, Mobile=@n15)
    False=ShadingPathSwitch(Default=@n17, Deferred=@n17, Forward=@n17, Mobile=@n19)
   ) {group=WPO, parameterName=Use_Glider_Lod_WPO_Fix}
  )
  n20 = MakeMaterialAttributes!(
   BaseColor
   Metallic
   Specular
   Roughness
   Anisotropy
   EmissiveColor
   Opacity
   OpacityMask
   Normal
   Tangent
   WorldPositionOffset
   SubsurfaceColor
   ClearCoat
   ClearCoatRoughness
   AmbientOcclusion
   Refraction
   CustomizedUVs_0
   CustomizedUVs_1
   CustomizedUVs_2
   CustomizedUVs_3
   CustomizedUVs_4
   CustomizedUVs_5
   CustomizedUVs_6
   CustomizedUVs_7
   PixelDepthOffset
   ShadingModel
   Displacement
  )
  MaterialExpressionReroute_1 = Reroute(None=@n20)
  n21 = VectorParameter() {parameterName=GalaxyDebug, defaultValue=<rgba 1 1 1 1>}
  n22 = MaterialFunctionCall!(
   MA=@MaterialExpressionReroute_1.Output
   SkinFX=MaterialFunctionCall(X=@n21.R, Y=@n21.R, Z=@n21.R, A=@n21.R) {materialFunction=&MakeFloat4}.Result
  ) {materialFunction=/Game/Packages/Fortress_SharedMaterials/Base_Material_Functions/MFC_GalaxyScreenSpace}
  MaterialExpressionReroute_2 = Reroute(None=@n22.Result)
  n23 = BreakMaterialAttributes!(Attr=@MaterialExpressionReroute_2.Attr)
  MaterialExpressionReroute_0 = Reroute(None=@n23.EmissiveColor)
  n24 = ParticleColor()
  n25 = VectorParameter() {parameterName=Emissive Color, defaultValue=<rgba 0.0198 0.0215 0.125 1>}
  n26 = VectorParameter() {parameterName=Emissive Color Night, defaultValue=<rgba 0.0123 0.0134 0.0775 1>}
  Add(
   A=@MaterialExpressionReroute_0.EmissiveColor
   B=LinearInterpolate(
    A=@n24.RGB
    B=MaterialFunctionCall(Color_Morning=@n25.RGB, Color_Day=@n25.RGB, Color_Evening=@n26.RGB, Color_Night=@n26.RGB) {materialFunction=&Color_Time_of_Day}.Result
    Alpha=ScalarParameter() {parameterName=UseEmissiveColor}
   )
  )
  n27 = Add(
   A=DynamicParameter() {paramNames [Time offset, Param2, PowerMultiplier, Param4]}.Time offset
   B=Add(A=Time(), B=ScalarParameter() {parameterName=TextureTimeOffset})
  )
  n28 = Panner(Coordinate=TextureCoordinate() {vTiling=0.5714}, Time=@n27) {speedX=-0.1, speedY=1}
  n29 = Multiply(A=Power(Base=TextureSample(UVs=@n28) {texture=/Game/Global/Textures/Default/Noise/T_Tiling_Noise_01}.RGB) {constExponent=0.9}) {constB=3}
  n30 = MaterialFunctionCall(Float2=TextureCoordinate()) {materialFunction=&BreakOutFloat2Components}
  n31 = TextureSample(UVs=MaterialFunctionCall(X=@n30.R, Y=OneMinus(None=@n30.G)) {materialFunction=&MakeFloat2}.Result) {texture=/Game/Global/Textures/Default/Masks_and_Gradients/T_Soft_Ramp_Vert_03}
  n32 = Time()
  n33 = Power(
   Base=@n31.RGB
   Exp=MaterialFunctionCall(Input=Sine(None=@n32), Input Low=Constant() {r=-1}, Input High=Constant() {r=1}, Target Low=Constant() {r=0.8}, Target High=Constant() {r=3}) {materialFunction=&RemapValueRange}.Result
  ) {constExponent=1}
  n34 = Clamp(None=Add(A=Multiply(A=@n29, B=@n33) {constB=2}, B=Power(Base=@n31.RGB) {constExponent=1}))
  Add(A=@MaterialExpressionReroute_0.EmissiveColor, B=@n34)
  n35 = DotProduct(A=CameraVectorWS(), B=VertexNormalWS())
  n36 = Clamp(None=Multiply(A=@n35))
  Clamp(None=Multiply(B=@n36))
  Composite!(Attr=@n22.Result)
  Composite!()
  DepthFade(FadeDistance=ScalarParameter() {defaultValue=1.5, parameterName=DepthFade})
  Multiply(A=@n24.A, B=LinearInterpolate(A=@n34, B=@n24.A, Alpha=ScalarParameter() {parameterName=UseFade}))
  n37 = Power(Base=Multiply(A=@n35), Exp=Constant() {r=2})
  n38 = Power(Base=@n35) {constExponent=0.5}
  Multiply(A=@n34, B=Add(A=Clamp(None=@n37) {maxDefault=0.8}, B=Add(A=@n38, B=@n36)) {constB=0})
  NamedRerouteDeclaration(None=@n29) {name=FireMask, nodeColor=<rgba 0.8281 1 0 1>}
  n39 = Panner(Coordinate=TextureCoordinate(), Time=@n27) {speedX=-0.01, speedY=0.3}
  n40 = TextureSample(UVs=Panner(Coordinate=TextureCoordinate(), Time=@n27) {speedX=-0.2, speedY=1}) {texture=&T_Tiling_Noise_01}
  NamedRerouteDeclaration(None=Multiply(
    A=TextureSample(UVs=@n39) {samplerType=SAMPLERTYPE_LinearColor, texture=/Game/Effects/Fort_Effects/BluGlow/Violentium/Violentium_Mask}.RGB
    B=LinearInterpolate(A=@n40.RGB) {constA=1}
   )) {name=FireWPO, nodeColor=<rgba 0 1 0.0156 1>}
  OneMinus(None=SphereMask(
    A=WorldPosition() {worldPositionShaderOffset=WPT_ExcludeAllShaderOffsets}.XYZ
    B=ObjectPositionWS()
    Radius=ScalarParameter() {defaultValue=45, parameterName=SphereRadius}
    Hardness=ScalarParameter() {defaultValue=1, parameterName=SphereHardness}
   ) {attenuationRadius=150, hardnessPercent=10})
  PinBase() {reroutePins [Attr {expression=&self:MaterialExpressionReroute_2}]}
  PinBase!(EmissiveColor=@n23.EmissiveColor) {reroutePins [EmissiveColor {expression=&self:MaterialExpressionReroute_0}]}
  PinBase!()
  PinBase!(Output=@n20) {reroutePins [Output {expression=&self:MaterialExpressionReroute_1}]}
  Power(Base=@n35, Exp=Constant() {r=1.5})
  n41 = TextureCoordinate() {uTiling=0}
  n42 = Power(Base=Multiply(A=ComponentMask(None=@n41) {g=true}) {constB=1.15}) {constExponent=0.5}
  StaticSwitchParameter(True=Multiply(A=OneMinus(None=@n42), B=NamedRerouteUsage())) {defaultValue=true, parameterName=USE_WPO}
 ]
}
